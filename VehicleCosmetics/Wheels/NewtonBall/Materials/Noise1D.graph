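# Material function in the NewtonBall cosmetic wheel folder that turns a scalar input, previewed as scaled Time, into procedural 1D noise.
# The sine of the input's floor gives each integer interval its own pseudo-random amplitude.
# That amplitude scales a polynomial of the fractional part that reaches zero at interval boundaries, avoiding texture lookups.
# An If on alternate intervals returns one on both unequal branches, so it does not flip the sign.
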MaterialFunction /VehicleCosmetics/Wheels/NewtonBall/Materials/Noise1D
graph {nodes [
  n1 = FunctionInput(Preview=Multiply(A=Time(), B=Constant() {r=3})) {inputType=FunctionInput_Scalar}
  n2 = Frac(None=@n1)
  n3 = Sine(None=Floor(None=@n1))
  n4 = Constant() {r=0.5}
  n5 = Multiply(A=Add(A=Multiply(A=@n3, B=@n4)), B=Constant() {r=16})
  n6 = Constant() {r=4}
  n7 = Multiply(A=Subtract(A=@n2, B=Constant() {r=1}), B=Multiply(A=Subtract(A=@n5, B=@n6), B=@n2))
  n8 = Constant()
  n9 = Subtract(A=Subtract(A=@n2, B=Constant() {r=1}), B=Constant() {r=1})
  n10 = Frac(None=Divide(A=@n1, B=Constant() {r=2}))
  n11 = Constant() {r=0.5}
  n12 = Constant()
  n13 = Multiply(A=@n2, B=If(A=Subtract(A=@n10, B=@n11), B=@n12, "A > B"=Constant() {r=1}, "A == B"=@n12, "A < B"=Constant() {r=1}))
  FunctionOutput(None=Multiply(A=Subtract(A=@n7, B=@n8), B=Multiply(A=@n9, B=@n13)))
 ]}
