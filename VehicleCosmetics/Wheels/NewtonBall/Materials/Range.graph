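# Material function in the Newton Ball wheel cosmetic's folder that linearly remaps an input from an old min-max range to a new one.
# It normalizes the input by subtracting Old Min and dividing by the old span, then scales by the new span and adds New Min.
# The result is unclamped, with no guard against a zero-width old range.
# Texture coordinates padded to three channels serve as the input's preview.
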MaterialFunction /VehicleCosmetics/Wheels/NewtonBall/Materials/Range
graph {nodes [
  n1 = TextureCoordinate()
  n2 = AppendVector(A=AppendVector(A=ComponentMask(None=@n1) {r=true}, B=ComponentMask(None=@n1) {g=true}), B=Constant())
  n3 = FunctionInput() {inputName=Old Min, sortPriority=2}
  n4 = Divide(
   A=Subtract(A=FunctionInput(Preview=@n2), B=@n3)
   B=Subtract(A=FunctionInput() {inputName=Old Max, sortPriority=1, previewValue=<1 1 1 1>}, B=@n3)
  )
  n5 = FunctionInput() {inputName=New Min, sortPriority=4}
  n6 = Subtract(A=FunctionInput() {inputName=New Max, sortPriority=3, previewValue=<1 1 1 1>}, B=@n5)
  FunctionOutput(None=Add(A=Multiply(A=@n4, B=@n6), B=@n5))
 ]}
