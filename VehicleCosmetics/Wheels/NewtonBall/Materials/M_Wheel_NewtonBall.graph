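# Material for the NewtonBall cosmetic vehicle wheel on a skeletal mesh, adding a two-color animated glow to the simple wheel parent function.
# The emissive mask's R channel drives a steady color, while its A channel drives a second color that blinks through time-driven noise, sine and a narrow range remap.
# Vertex color green blends this glow over the parent emissive, then the attributes pass through vehicle damage and de-rez transition functions.
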Material /VehicleCosmetics/Wheels/NewtonBall/Materials/M_Wheel_NewtonBall
graph {
 properties {
  bAutomaticallySetUsageInEditor=false
  bUseMaterialAttributes=true
  bUsedWithMeshDeformer=true
  bUsedWithSkeletalMesh=true
  pixelDepthOffsetMode=PDOM_Legacy
  previewMesh=/VehicleCosmetics/Wheels/NewtonBall/New/Mesh/SK_Wheel_Newtonball
  usedShadingModels=Default Lit
 }
 nodes [
  n1 = MaterialFunctionCall() {materialFunction=&MF_WheelParent_Simple}
  ComponentMask(None=@n1.VertexColorR) {g=true}
  n2 = GetMaterialAttributes(None=@n1.MaterialAttributes) {attributeGetTypes [#1]}
  n3 = LinearInterpolate(A=@n2.Emissive Color, B=NamedRerouteUsage(), Alpha=VertexColor().G)
  MaterialFunctionCall(In=MaterialFunctionCall(Material=SetMaterialAttributes(MaterialAttributes=@n2.MaterialAttributes, Emissive Color=@n3) {attributeSetTypes [#1]}) {materialFunction=&MF_BaseVehicle_Damage}.Result) {materialFunction=&MF_VehicleDeRez_Transition}
  Multiply()
  n4 = ScalarParameter() {defaultValue=1, group=5_Special Effects, parameterName=Brightness 1}
  n5 = TextureSampleParameter2D() {
   group=5_Special Effects
   parameterName=Emissive Mask
   samplerType=SAMPLERTYPE_LinearColor
   texture=/VehicleCosmetics/Wheels/NewtonBall/Textures/T_Newtonball_E
  }
  n6 = Multiply(A=VectorParameter() {group=5_Special Effects, parameterName=Custom Color 1, defaultValue=<rgba 1 0.0925 0.0044 1>}.RGB, B=@n5.R)
  n7 = Constant()
  n8 = Multiply(
   A=ScalarParameter() {defaultValue=1, group=5_Special Effects, parameterName=Brightness 2}
   B=Multiply(A=VectorParameter() {group=5_Special Effects, parameterName=Custom Color 2, defaultValue=<rgba 0 0.721 1 1>}.RGB, B=@n5.A)
  )
  n9 = MaterialFunctionCall!(In=Multiply(A=Time(), B=ScalarParameter() {defaultValue=0.2, group=5_Special Effects, parameterName=Light Blink})) {materialFunction=&/Noise1D}
  n10 = Clamp(None=MaterialFunctionCall!(
    In=Sine(None=@n9.Result)
    Old Max=ScalarParameter() {defaultValue=0.1505, group=5_Special Effects, parameterName=Light Blink Off/Max}
    Old Min=ScalarParameter() {defaultValue=0.15, group=5_Special Effects, parameterName=Light Blink Off/Min}
    New Max=Constant() {r=1}
    New Min=Constant()
   ) {materialFunction=&/Range}.Result)
  NamedRerouteDeclaration(None=Add(A=Multiply(A=@n4, B=@n6), B=LinearInterpolate(A=@n7, B=@n8, Alpha=@n10))) {name=Emissive, nodeColor=<rgba 1 0.75 0 1>}
  Subtract()
 ]
}
